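# Parent material for placeholder shoe cosmetics, built on material attributes and the 2023 character parent functions.
# PreFX output and mask feed named reroutes around an optional paint stage, then MF_Sock, MF_Holo_Placeholder and PostFX finish the chain.
# The OverrideBasePaint static switch replaces base color with a two-color blend driven by roughness raised to an exponent, multiplied by a green mask channel.
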Material /Game/Global/Materials/Cosmetics/Kicks/Shoes/Materials/Placeholder/M_Shoes_Placeholder_Parent
graph {
 properties {
  bUseMaterialAttributes=true
  bUsedWithClothing=true
  bUsedWithMeshDeformer=true
  bUsedWithMorphTargets=true
  bUsedWithNanite=true
  bUsedWithNiagaraMeshParticles=true
  bUsedWithSkeletalMesh=true
  pixelDepthOffsetMode=PDOM_Legacy
  shadingModel=MSM_FromMaterialExpression
  usedShadingModels=Default Lit | Subsurface | Clear Coat | Subsurface Profile | Cloth
 }
 nodes [
  n1 = NamedRerouteUsage()
  n2 = MaterialFunctionCall() {materialFunction=&MF_CharacterParent_2023_PreFX}
  MaterialFunctionCall(
   Material=MaterialFunctionCall!(InMaterial=MaterialFunctionCall!(InOpacity, Material=@n1, Mask=@n2.Tech Art Mask) {materialFunction=/Game/Global/Materials/Cosmetics/Kicks/Socks/Materials/MF_Sock}.Result) {materialFunction=&/MF_Holo_Placeholder}.Result
   M=@n2.M
   Pre Skinned Local Position=@n2.M
   Distance Blend=@n2.M
   Fresnel=@n2.M
  ) {materialFunction=&MF_CharacterParent_2023_PostFX}
  NamedRerouteDeclaration(None=@n2.Result) {name=InMaterial, nodeColor=<rgba 0 0.6484 1 1>}
  NamedRerouteDeclaration(None=@n2.M) {name=InMask, nodeColor=<rgba 0.9688 1 0 1>}
  n3 = GetMaterialAttributes(None=NamedRerouteUsage()) {attributeGetTypes [#1]}
  n4 = ScalarParameter() {defaultValue=2.2141, parameterName=Exp}
  n5 = Multiply(
   A=LinearInterpolate(
    A=VectorParameter() {defaultValue=<rgba 0.0271 0.0269 0.0312 1>}.RGB
    B=VectorParameter() {parameterName=Param_1, defaultValue=<rgba 0.5753 0.6267 0.6771 1>}.RGB
    Alpha=Power(Base=@n3.Roughness, Exp=@n4)
   )
   B=ComponentMask(None=NamedRerouteUsage()) {g=true}
  )
  NamedRerouteDeclaration(None=StaticSwitchParameter(
    True=SetMaterialAttributes!(MaterialAttributes=@n3.MaterialAttributes, Base Color=@n5) {attributeSetTypes [#2]}
    False=NamedRerouteUsage()
   ) {parameterName=OverrideBasePaint}) {name=OutMaterial, nodeColor=<rgba 0.7109 0 1 1>}
 ]
}
